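# Physics asset for the Tender Defender male medium player body cosmetic, giving physics bodies to its tail root and pelvis.
# It has two capsule bodies, an offset, rotated one on dyn_tail_root and a zero-length, effectively spherical one on pelvis.
# A single constraint joins the tail root to the pelvis with moderate, equal swing and twist angle limits.
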PhysicsAsset /BRCosmetics/Characters/Player/Male/Medium/Bodies/M_MED_Tender_Defender/Meshes/M_MED_Tender_Defender_Physics
bodies {
 dyn_tail_root [{shapeType=Capsule, length=1.991, radius=3.7229, center=<-0.3342 -4.3802 -0.0523>, rotation=<rot 0 4.1654 -83.5121>}]
 pelvis [{shapeType=Capsule, length=0, radius=5}]
}
constraints [{bone1Name=dyn_tail_root, bone2Name=pelvis, swing1LimitDegrees=45, swing2LimitDegrees=45, twistLimitDegrees=45}]
