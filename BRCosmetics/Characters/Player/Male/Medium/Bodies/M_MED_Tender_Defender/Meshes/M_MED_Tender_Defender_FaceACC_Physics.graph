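# Physics asset for the face accessory mesh of the male medium Tender Defender player cosmetic, giving its chin and hair bones collision bodies and joint limits.
# A large zero-length head capsule serves as the hub, with smaller capsules on one chin bone and five dyn_hair bones.
# Each dyn bone constrains directly to the head rather than chaining, with individual swing limits and a uniformly tight twist limit.
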PhysicsAsset /BRCosmetics/Characters/Player/Male/Medium/Bodies/M_MED_Tender_Defender/Meshes/M_MED_Tender_Defender_FaceACC_Physics
bodies {
 dyn_chin_01 [{shapeType=Capsule, length=0, radius=2, center=<3 0 0>, rotation=<rot 0 0 5>}]
 dyn_hair_01_a [{
   shapeType=Capsule
   length=4.4327
   radius=2.6896
   shapeName=dyn_hair_01_a_capsule
   center=<0.3583 -7.4561 -0.2273>
   rotation=<rot 0 0.0149 -88.1455>
  }]
 dyn_hair_01_b [{
   shapeType=Capsule
   length=4.0023
   radius=2.6085
   shapeName=dyn_hair_01_b_capsule
   center=<0.241 -5.0746 -1.05>
   rotation=<rot 0 -179.9222 -67.8964>
  }]
 dyn_hair_01_c [{
   shapeType=Capsule
   length=2.4975
   radius=1.9245
   shapeName=dyn_hair_01_c_capsule
   center=<0.305 -5.5391 0.5891>
   rotation=<rot 0 1.8514 -77.7455>
  }]
 dyn_hair_03_a [{
   shapeType=Capsule
   length=5.9339
   radius=2.7217
   shapeName=dyn_hair_03_a_capsule
   center=<0.3261 -6.4565 -0.3804>
   rotation=<rot 0 0.5525 -88.8786>
  }]
 dyn_hair_03_b [{
   shapeType=Capsule
   length=4.7055
   radius=3.6531
   shapeName=dyn_hair_03_b_capsule
   center=<0.0402 -4.3919 4.191>
   rotation=<rot -36.3333 -90 -0.6534>
  }]
 head [{shapeType=Capsule, length=0, radius=11.3907, center=<7.3577 1.3763 -0.6724>, rotation=<rot 0 -89.6041 -89.7752>}]
}
constraints [
 {bone1Name=dyn_chin_01, bone2Name=head, swing1LimitDegrees=45, swing2LimitDegrees=45, twistLimitDegrees=2}
 {bone1Name=dyn_hair_01_a, bone2Name=head, swing1LimitDegrees=45, swing2LimitDegrees=20, twistLimitDegrees=2}
 {bone1Name=dyn_hair_01_b, bone2Name=head, swing1LimitDegrees=30, swing2LimitDegrees=20, twistLimitDegrees=2}
 {bone1Name=dyn_hair_01_c, bone2Name=head, swing1LimitDegrees=10, swing2LimitDegrees=20, twistLimitDegrees=2}
 {bone1Name=dyn_hair_03_a, bone2Name=head, swing1LimitDegrees=20, swing2LimitDegrees=50, twistLimitDegrees=2}
 {bone1Name=dyn_hair_03_b, bone2Name=head, swing1LimitDegrees=10, swing2LimitDegrees=20, twistLimitDegrees=2}
]
