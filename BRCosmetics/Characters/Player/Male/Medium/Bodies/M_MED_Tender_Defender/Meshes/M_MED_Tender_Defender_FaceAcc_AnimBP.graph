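# Anim Blueprint for the Tender Defender face accessory, copying the attached parent's pose and adding secondary motion.
# The cached pose runs through a head constraint on dyn_hair_01, AnimDynamics for hair, chin and LOD-limited feathers, then RigidBody.
# When the playlist uses custom character parts, checked at init, the output instead shrinks the head bone to near zero.
# Several dynamics chains are disconnected, and the vehicle-driven Vehicle Off value is unused.
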AnimBlueprint /BRCosmetics/Characters/Player/Male/Medium/Bodies/M_MED_Tender_Defender/Meshes/M_MED_Tender_Defender_FaceAcc_AnimBP
target_skeleton=&/M_MED_Tender_Defender_FaceAcc_Skeleton
graphs [
 AnimGraph {nodes [
   Anim Dynamics - Bone: dyn_jaw_lower(Alpha=1, ExternalForce="0,0,0") {
    alpha=0.9
    angularSpringConstant=60
    bAngularSpring=true
    bGravityOverrideInSimSpace=true
    gravityScale=0.5
    numSolverIterationsPostUpdate=4
    numSolverIterationsPreUpdate=16
    simulationSpace=BoneRelative
    alphaBoolBlend {blendOption=Linear}
    boundBone {boneName=dyn_jaw_lower}
    physicsBodyDefinitions [{
      sphereCollisionRadius=0
      boundBone {boneName=dyn_jaw_lower}
      boxExtents=<15 5 5>
      constraintSetup {angularTargetAxis=AxisZ, twistAxis=AxisY, angularLimitsMax=<10 0 0>, angularLimitsMin=<-80 0 0>, angularTarget=<0 0 1>}
      localJointOffset=<0 0 -1>
     }]
    relativeSpaceBone {boneName=root}
   }
   n1 = Anim Dynamics - Bone: dyn_hair_01_a(Alpha=0.25, ExternalForce="0,0,0") {
    angularDampingOverride=0.9999
    angularSpringConstant=80
    bAngularSpring=true
    bGravityOverrideInSimSpace=true
    bOverrideAngularDamping=true
    gravityScale=0.85
    numSolverIterationsPostUpdate=4
    numSolverIterationsPreUpdate=16
    simulationSpace=BoneRelative
    alphaBoolBlend {blendOption=Linear}
    boundBone {boneName=dyn_hair_01_a}
    physicsBodyDefinitions [{
      sphereCollisionRadius=0
      boundBone {boneName=dyn_hair_01_a}
      boxExtents=<15 5 5>
      constraintSetup {angularLimitsMax=<1 45 15>, angularLimitsMin=<-1 -45 -15>, angularTarget=<0 0 0>}
      localJointOffset=<0 -2 3>
     }]
    relativeSpaceBone {boneName=root}
   }
   n2 = Anim Dynamics - Bone: dyn_hair_01_b(ComponentPose=@n1, Alpha=0.25, ExternalForce="0,0,0") {
    angularDampingOverride=0.9999
    angularSpringConstant=80
    bAngularSpring=true
    bGravityOverrideInSimSpace=true
    bOverrideAngularDamping=true
    gravityScale=0.85
    numSolverIterationsPostUpdate=4
    numSolverIterationsPreUpdate=16
    simulationSpace=BoneRelative
    alphaBoolBlend {blendOption=Linear}
    boundBone {boneName=dyn_hair_01_b}
    physicsBodyDefinitions [{
      sphereCollisionRadius=0
      boundBone {boneName=dyn_hair_01_b}
      boxExtents=<15 5 5>
      constraintSetup {angularLimitsMax=<1 45 15>, angularLimitsMin=<-1 -45 -15>, angularTarget=<0 0 0>}
      localJointOffset=<0 -2 3>
     }]
    relativeSpaceBone {boneName=root}
   }
   Anim Dynamics - Bone: dyn_hair_01_c(ComponentPose=@n2, Alpha=0.25, ExternalForce="0,0,0") {
    angularDampingOverride=0.9999
    angularSpringConstant=80
    bAngularSpring=true
    bGravityOverrideInSimSpace=true
    bOverrideAngularDamping=true
    gravityScale=0.85
    numSolverIterationsPostUpdate=4
    numSolverIterationsPreUpdate=16
    simulationSpace=BoneRelative
    alphaBoolBlend {blendOption=Linear}
    boundBone {boneName=dyn_hair_01_c}
    physicsBodyDefinitions [{
      sphereCollisionRadius=0
      boundBone {boneName=dyn_hair_01_c}
      boxExtents=<15 5 5>
      constraintSetup {angularLimitsMax=<1 45 15>, angularLimitsMin=<-1 -45 -15>, angularTarget=<0 0 0>}
      localJointOffset=<0 -2 3>
     }]
    relativeSpaceBone {boneName=root}
   }
   n3 = Anim Dynamics - Bone: dyn_hair_01_a(Alpha=0.25, ExternalForce="0,0,0") {
    alpha=0.9
    angularDampingOverride=0.99
    angularSpringConstant=80
    bAngularSpring=true
    bGravityOverrideInSimSpace=true
    bOverrideAngularDamping=true
    gravityScale=0.85
    numSolverIterationsPostUpdate=4
    numSolverIterationsPreUpdate=16
    simulationSpace=BoneRelative
    alphaBoolBlend {blendOption=Linear}
    boundBone {boneName=dyn_hair_01_a}
    physicsBodyDefinitions [{
      sphereCollisionRadius=0
      boundBone {boneName=dyn_hair_01_a}
      boxExtents=<15 5 5>
      constraintSetup {angularTargetAxis=AxisZ, twistAxis=AxisY, angularLimitsMax=<15 0 89>, angularLimitsMin=<-15 0 -89>, angularTarget=<0 0 1>}
      localJointOffset=<0 2 6>
     }]
    relativeSpaceBone {boneName=root}
   }
   n4 = Anim Dynamics - Bone: dyn_hair_01_b(ComponentPose=@n3, Alpha=0.25, ExternalForce="0,0,0") {
    alpha=0.9
    angularDampingOverride=0.99
    angularSpringConstant=80
    bAngularSpring=true
    bGravityOverrideInSimSpace=true
    bOverrideAngularDamping=true
    gravityScale=0.85
    numSolverIterationsPostUpdate=4
    numSolverIterationsPreUpdate=16
    simulationSpace=BoneRelative
    alphaBoolBlend {blendOption=Linear}
    boundBone {boneName=dyn_hair_01_b}
    physicsBodyDefinitions [{
      sphereCollisionRadius=0
      boundBone {boneName=dyn_hair_01_b}
      boxExtents=<15 5 5>
      constraintSetup {angularTargetAxis=AxisZ, twistAxis=AxisY, angularLimitsMax=<15 0 89>, angularLimitsMin=<-15 0 -89>, angularTarget=<0 0 1>}
      localJointOffset=<0 2 6>
     }]
    relativeSpaceBone {boneName=root}
   }
   n5 = Anim Dynamics - Bone: dyn_hair_01_c(ComponentPose=@n4, Alpha=0.25, ExternalForce="0,0,0") {
    alpha=0.9
    angularDampingOverride=0.9999
    angularSpringConstant=80
    bAngularSpring=true
    bGravityOverrideInSimSpace=true
    bOverrideAngularDamping=true
    gravityScale=0.85
    simulationSpace=BoneRelative
    alphaBoolBlend {blendOption=Linear}
    boundBone {boneName=dyn_hair_01_c}
    physicsBodyDefinitions [{
      sphereCollisionRadius=0
      boundBone {boneName=dyn_hair_01_c}
      boxExtents=<15 5 5>
      constraintSetup {angularTargetAxis=AxisZ, twistAxis=AxisY, angularLimitsMax=<15 0 89>, angularLimitsMin=<-15 0 -89>, angularTarget=<0 0 1>}
      localJointOffset=<0 1 4>
     }]
    relativeSpaceBone {boneName=neck_01}
   }
   n6 = Anim Dynamics - Bone: dyn_hair_01(ComponentPose=@n5, Alpha=0) {
    alpha=0.9
    angularSpringConstant=80
    bAngularSpring=true
    bGravityOverrideInSimSpace=true
    numSolverIterationsPostUpdate=4
    numSolverIterationsPreUpdate=16
    simulationSpace=BoneRelative
    alphaBoolBlend {blendOption=Linear}
    boundBone {boneName=dyn_hair_01}
    physicsBodyDefinitions [{
      sphereCollisionRadius=0
      boundBone {boneName=dyn_hair_01}
      boxExtents=<15 5 5>
      constraintSetup {angularTargetAxis=AxisZ, twistAxis=AxisY, angularLimitsMax=<15 0 89>, angularLimitsMin=<-15 0 -89>, angularTarget=<0 0 1>}
      localJointOffset=<0 0 3>
     }]
    relativeSpaceBone {boneName=root}
   }
   Anim Dynamics - Bone: dyn_hair_03(ComponentPose=@n6, Alpha=0, ExternalForce="0,0,0") {
    alpha=0.9
    angularSpringConstant=80
    bAngularSpring=true
    bGravityOverrideInSimSpace=true
    numSolverIterationsPostUpdate=4
    numSolverIterationsPreUpdate=16
    simulationSpace=BoneRelative
    alphaBoolBlend {blendOption=Linear}
    boundBone {boneName=dyn_hair_03}
    physicsBodyDefinitions [{
      sphereCollisionRadius=0
      boundBone {boneName=dyn_hair_03}
      boxExtents=<15 5 5>
      constraintSetup {angularTargetAxis=AxisZ, twistAxis=AxisY, angularLimitsMax=<15 0 89>, angularLimitsMin=<-15 0 -89>, angularTarget=<0 0 1>}
      localJointOffset=<0 0 3>
     }]
    relativeSpaceBone {boneName=root}
   }
   RigidBody() {
    bOverrideWorldGravity=true
    overridePhysicsAsset=&/M_MED_Tender_Defender_FaceACC_Physics
    simulationSpace=BaseBoneSpace
    alphaBoolBlend {blendOption=Linear}
    componentLinearAccScale=<1 1 1>
    componentLinearVelScale=<1 1 1>
    overrideWorldGravity {z=-980}
   }
   n7 = "Transform (Modify) Bone - Bone: head"!(ComponentPose=Local To Component(LocalPose=Use cached pose 'PreScale Cache'()), Alpha=1) -> (Pose) {scaleMode=BMM_Replace, alphaBoolBlend {blendOption=Linear}, boneToModify {boneName=head}, scale=<0.01 0.01 0.01>}
   n8 = "Constraint - dyn_hair_01 (P)"(ComponentPose=Local To Component(LocalPose=Use cached pose 'PreScale Cache'()), Alpha=1, ConstraintWeights_0=1) {
    alphaBoolBlend {blendOption=Linear}
    boneToModify {boneName=dyn_hair_01}
    constraintSetup [{targetBone {boneName=head}}]
    constraintWeights [1]
   }
   n9 = Anim Dynamics - Bone: dyn_hair_01_a(ComponentPose=@n8, Alpha=0.25, ExternalForce="0,0,0") {
    alpha=0.9
    angularDampingOverride=0.99
    angularSpringConstant=80
    bAngularSpring=true
    bGravityOverrideInSimSpace=true
    bLinearSpring=true
    bOverrideAngularDamping=true
    gravityScale=0.85
    linearSpringConstant=20
    numSolverIterationsPostUpdate=4
    numSolverIterationsPreUpdate=16
    simulationSpace=BoneRelative
    alphaBoolBlend {blendOption=Linear}
    alphaScaleBiasClamp {bClampResult=true, clampMax=0.25}
    boundBone {boneName=dyn_hair_01_a}
    physicsBodyDefinitions [{
      sphereCollisionRadius=0
      boundBone {boneName=dyn_hair_01_a}
      boxExtents=<15 5 5>
      constraintSetup {angularTargetAxis=AxisZ, twistAxis=AxisY, angularLimitsMax=<15 0 89>, angularLimitsMin=<-15 0 -89>, angularTarget=<0 0 1>}
      localJointOffset=<0 2 -2>
     }]
    relativeSpaceBone {boneName=pelvis}
   }
   n10 = Anim Dynamics - Bone: dyn_hair_01_b(ComponentPose=@n9, Alpha=0.25, ExternalForce="0,0,0") {
    alpha=0.9
    angularDampingOverride=0.99
    angularSpringConstant=80
    bAngularSpring=true
    bGravityOverrideInSimSpace=true
    bLinearSpring=true
    bOverrideAngularDamping=true
    gravityScale=0.85
    linearSpringConstant=20
    numSolverIterationsPostUpdate=4
    numSolverIterationsPreUpdate=16
    simulationSpace=BoneRelative
    alphaBoolBlend {blendOption=Linear}
    alphaScaleBiasClamp {bClampResult=true, clampMax=0.25}
    boundBone {boneName=dyn_hair_01_b}
    physicsBodyDefinitions [{
      sphereCollisionRadius=0
      boundBone {boneName=dyn_hair_01_b}
      boxExtents=<15 5 5>
      constraintSetup {angularTargetAxis=AxisZ, twistAxis=AxisY, angularLimitsMax=<15 0 89>, angularLimitsMin=<-15 0 -89>, angularTarget=<0 0 1>}
      localJointOffset=<0 2 -2>
     }]
    relativeSpaceBone {boneName=pelvis}
   }
   n11 = Anim Dynamics - Bone: dyn_hair_01_c(ComponentPose=@n10, Alpha=0.25, ExternalForce="0,0,0") {
    alpha=0.9
    angularDampingOverride=0.9999
    angularSpringConstant=80
    bAngularSpring=true
    bGravityOverrideInSimSpace=true
    bLinearSpring=true
    bOverrideAngularDamping=true
    gravityScale=0.85
    linearSpringConstant=20
    simulationSpace=BoneRelative
    alphaBoolBlend {blendOption=Linear}
    alphaScaleBiasClamp {bClampResult=true, clampMax=0.25}
    boundBone {boneName=dyn_hair_01_c}
    physicsBodyDefinitions [{
      sphereCollisionRadius=0
      boundBone {boneName=dyn_hair_01_c}
      boxExtents=<15 5 5>
      constraintSetup {angularTargetAxis=AxisZ, twistAxis=AxisY, angularLimitsMax=<15 0 89>, angularLimitsMin=<-15 0 -89>, angularTarget=<0 0 1>}
      localJointOffset=<0 2 -6>
     }]
    relativeSpaceBone {boneName=pelvis}
   }
   n12 = Anim Dynamics - Bone: dyn_hair_03_a(ComponentPose=@n11, Alpha=0.25, ExternalForce="0,0,0") {
    alpha=0.9
    angularDampingOverride=0.7
    angularSpringConstant=80
    bAngularSpring=true
    bGravityOverrideInSimSpace=true
    bOverrideAngularDamping=true
    gravityScale=0.85
    numSolverIterationsPostUpdate=4
    numSolverIterationsPreUpdate=16
    simulationSpace=BoneRelative
    alphaBoolBlend {blendOption=Linear}
    alphaScaleBiasClamp {bClampResult=true, clampMax=0.25}
    boundBone {boneName=dyn_hair_03_a}
    physicsBodyDefinitions [{
      sphereCollisionRadius=0
      boundBone {boneName=dyn_hair_03_a}
      boxExtents=<15 5 5>
      constraintSetup {angularTargetAxis=AxisZ, twistAxis=AxisY, angularLimitsMax=<15 0 89>, angularLimitsMin=<-15 0 -89>, angularTarget=<0 0 1>}
      localJointOffset=<0 2 6>
     }]
    relativeSpaceBone {boneName=root}
   }
   n13 = Anim Dynamics - Bone: dyn_hair_03_b(ComponentPose=@n12, Alpha=0.25, ExternalForce="0,0,0") {
    alpha=0.9
    angularDampingOverride=0.7
    angularSpringConstant=80
    bAngularSpring=true
    bGravityOverrideInSimSpace=true
    bOverrideAngularDamping=true
    gravityScale=0.85
    numSolverIterationsPostUpdate=4
    numSolverIterationsPreUpdate=16
    simulationSpace=BoneRelative
    alphaBoolBlend {blendOption=Linear}
    alphaScaleBiasClamp {bClampResult=true, clampMax=0.25}
    boundBone {boneName=dyn_hair_03_b}
    physicsBodyDefinitions [{
      sphereCollisionRadius=0
      boundBone {boneName=dyn_hair_03_b}
      boxExtents=<15 5 5>
      constraintSetup {angularTargetAxis=AxisZ, twistAxis=AxisY, angularLimitsMax=<15 0 89>, angularLimitsMin=<-15 0 -89>, angularTarget=<0 0 1>}
      localJointOffset=<0 2 2>
     }]
    relativeSpaceBone {boneName=root}
   }
   n14 = Anim Dynamics - Bone: dyn_chin_01(ComponentPose=@n13, Alpha=1, ExternalForce="0,0,0") {
    angularSpringConstant=1500
    bGravityOverrideInSimSpace=true
    bUseGravityOverride=true
    gravityScale=0.25
    simulationSpace=BoneRelative
    alphaBoolBlend {blendOption=Linear}
    boundBone {boneName=dyn_chin_01}
    componentLinearAccScale=<0.1 0.1 0.1>
    componentLinearVelScale=<0.1 0.1 0.1>
    gravityOverride {z=-980}
    physicsBodyDefinitions [{
      sphereCollisionRadius=0
      boundBone {boneName=dyn_chin_01}
      boxExtents=<8 20 20>
      constraintSetup {angularLimitsMax=<0.2 20 10>, angularLimitsMin=<-1 -5 -8>, angularTarget=<0 0 0>}
      localJointOffset=<5 0 0>
     }]
   }
   n15 = Anim Dynamics - Bone: dyn_chin_02(ComponentPose=@n14, Alpha=1, ExternalForce="0,0,0") {
    angularSpringConstant=1500
    bGravityOverrideInSimSpace=true
    bUseGravityOverride=true
    gravityScale=0.25
    simulationSpace=BoneRelative
    alphaBoolBlend {blendOption=Linear}
    boundBone {boneName=dyn_chin_02}
    componentLinearAccScale=<0.1 0.1 0.1>
    componentLinearVelScale=<0.1 0.1 0.1>
    gravityOverride {z=-980}
    physicsBodyDefinitions [{
      sphereCollisionRadius=0
      boundBone {boneName=dyn_chin_02}
      boxExtents=<8 20 20>
      constraintSetup {angularLimitsMax=<0.2 20 10>, angularLimitsMin=<-1 -5 -8>, angularTarget=<0 0 0>}
      localJointOffset=<5 0 0>
     }]
   }
   n16 = Anim Dynamics - Bone: dyn_feather_01(ComponentPose=@n15, Alpha=0.25, ExternalForce="0,0,0") {
    angularDampingOverride=0.98
    bGravityOverrideInSimSpace=true
    bOverrideAngularDamping=true
    bOverrideLinearDamping=true
    gravityScale=0.1
    lODThreshold=1
    linearDampingOverride=0.98
    numSolverIterationsPostUpdate=4
    numSolverIterationsPreUpdate=16
    simulationSpace=BoneRelative
    alphaBoolBlend {blendOption=Linear}
    boundBone {boneName=dyn_feather_01}
    physicsBodyDefinitions [{
      sphereCollisionRadius=0
      boundBone {boneName=dyn_feather_01}
      boxExtents=<5 5 5>
      constraintSetup {angularLimitsMax=<20 60 20>, angularLimitsMin=<-20 -5 -20>, angularTarget=<0 0 0>}
      localJointOffset=<-0.1 0 0>
     }]
    relativeSpaceBone {boneName=pelvis}
   }
   n17 = Anim Dynamics - Bone: dyn_feather_02(ComponentPose=@n16, Alpha=0.25, ExternalForce="0,0,0") {
    angularDampingOverride=0.98
    bGravityOverrideInSimSpace=true
    bOverrideAngularDamping=true
    bOverrideLinearDamping=true
    gravityScale=0.1
    lODThreshold=1
    linearDampingOverride=0.98
    numSolverIterationsPostUpdate=4
    numSolverIterationsPreUpdate=16
    simulationSpace=BoneRelative
    alphaBoolBlend {blendOption=Linear}
    boundBone {boneName=dyn_feather_02}
    physicsBodyDefinitions [{
      sphereCollisionRadius=0
      boundBone {boneName=dyn_feather_02}
      boxExtents=<5 5 5>
      constraintSetup {angularLimitsMax=<20 60 20>, angularLimitsMin=<-20 -5 -20>, angularTarget=<0 0 0>}
      localJointOffset=<0 0 0.5>
     }]
    relativeSpaceBone {boneName=pelvis}
   }
   n18 = Anim Dynamics - Bone: dyn_feather_03(ComponentPose=@n17, Alpha=0.25, ExternalForce="0,0,0") {
    angularDampingOverride=0.98
    bGravityOverrideInSimSpace=true
    bOverrideAngularDamping=true
    bOverrideLinearDamping=true
    gravityScale=0.1
    lODThreshold=1
    linearDampingOverride=0.98
    numSolverIterationsPostUpdate=4
    numSolverIterationsPreUpdate=16
    simulationSpace=BoneRelative
    alphaBoolBlend {blendOption=Linear}
    boundBone {boneName=dyn_feather_03}
    physicsBodyDefinitions [{
      sphereCollisionRadius=0
      boundBone {boneName=dyn_feather_03}
      boxExtents=<5 5 5>
      constraintSetup {angularLimitsMax=<20 15 20>, angularLimitsMin=<-20 -15 -20>, angularTarget=<0 0 0>}
      localJointOffset=<0 0 0.5>
     }]
    relativeSpaceBone {boneName=pelvis}
   }
   n19 = Anim Dynamics - Bone: dyn_feather_04(ComponentPose=@n18, Alpha=0.25, ExternalForce="0,0,0") {
    angularDampingOverride=0.98
    bGravityOverrideInSimSpace=true
    bOverrideAngularDamping=true
    bOverrideLinearDamping=true
    gravityScale=0.1
    lODThreshold=1
    linearDampingOverride=0.98
    numSolverIterationsPostUpdate=4
    numSolverIterationsPreUpdate=16
    simulationSpace=BoneRelative
    alphaBoolBlend {blendOption=Linear}
    boundBone {boneName=dyn_feather_04}
    physicsBodyDefinitions [{
      sphereCollisionRadius=0
      boundBone {boneName=dyn_feather_04}
      boxExtents=<5 5 5>
      constraintSetup {angularLimitsMax=<20 60 20>, angularLimitsMin=<-20 -5 -20>, angularTarget=<0 0 0>}
      localJointOffset=<0 0 -0.5>
     }]
    relativeSpaceBone {boneName=pelvis}
   }
   n20 = Anim Dynamics - Bone: dyn_feather_05(ComponentPose=@n19, Alpha=0.25, ExternalForce="0,0,0") {
    angularDampingOverride=0.98
    bGravityOverrideInSimSpace=true
    bOverrideAngularDamping=true
    bOverrideLinearDamping=true
    gravityScale=0.1
    lODThreshold=1
    linearDampingOverride=0.98
    numSolverIterationsPostUpdate=4
    numSolverIterationsPreUpdate=16
    simulationSpace=BoneRelative
    alphaBoolBlend {blendOption=Linear}
    boundBone {boneName=dyn_feather_05}
    physicsBodyDefinitions [{
      sphereCollisionRadius=0
      boundBone {boneName=dyn_feather_05}
      boxExtents=<5 5 5>
      constraintSetup {angularLimitsMax=<20 60 20>, angularLimitsMin=<-20 -5 -20>, angularTarget=<0 0 0>}
      localJointOffset=<0 0 -0.5>
     }]
    relativeSpaceBone {boneName=pelvis}
   }
   n21 = Anim Dynamics - Bone: dyn_feather_06(ComponentPose=@n20, Alpha=0.25, ExternalForce="0,0,0") {
    angularDampingOverride=0.98
    bGravityOverrideInSimSpace=true
    bOverrideAngularDamping=true
    bOverrideLinearDamping=true
    gravityScale=0.1
    lODThreshold=1
    linearDampingOverride=0.98
    numSolverIterationsPostUpdate=4
    numSolverIterationsPreUpdate=16
    simulationSpace=BoneRelative
    alphaBoolBlend {blendOption=Linear}
    boundBone {boneName=dyn_feather_06}
    physicsBodyDefinitions [{
      sphereCollisionRadius=0
      boundBone {boneName=dyn_feather_06}
      boxExtents=<5 5 5>
      constraintSetup {angularLimitsMax=<20 60 20>, angularLimitsMin=<-20 -5 -20>, angularTarget=<0 0 0>}
      localJointOffset=<0 0 -0.5>
     }]
    relativeSpaceBone {boneName=pelvis}
   }
   n22 = Anim Dynamics - Bone: dyn_feather_07(ComponentPose=@n21, Alpha=0.25, ExternalForce="0,0,0") {
    angularDampingOverride=0.98
    bGravityOverrideInSimSpace=true
    bOverrideAngularDamping=true
    bOverrideLinearDamping=true
    gravityScale=0.1
    lODThreshold=1
    linearDampingOverride=0.98
    numSolverIterationsPostUpdate=4
    numSolverIterationsPreUpdate=16
    simulationSpace=BoneRelative
    alphaBoolBlend {blendOption=Linear}
    boundBone {boneName=dyn_feather_07}
    physicsBodyDefinitions [{
      sphereCollisionRadius=0
      boundBone {boneName=dyn_feather_07}
      boxExtents=<5 5 5>
      constraintSetup {angularLimitsMax=<20 60 20>, angularLimitsMin=<-20 -5 -20>, angularTarget=<0 0 0>}
      localJointOffset=<0 0 -0.5>
     }]
    relativeSpaceBone {boneName=pelvis}
   }
   n23 = Anim Dynamics - Bone: dyn_feather_08(ComponentPose=@n22, Alpha=0.25, ExternalForce="0,0,0") {
    angularDampingOverride=0.98
    bGravityOverrideInSimSpace=true
    bOverrideAngularDamping=true
    bOverrideLinearDamping=true
    gravityScale=0.1
    lODThreshold=1
    linearDampingOverride=0.98
    numSolverIterationsPostUpdate=4
    numSolverIterationsPreUpdate=16
    simulationSpace=BoneRelative
    alphaBoolBlend {blendOption=Linear}
    boundBone {boneName=dyn_feather_08}
    physicsBodyDefinitions [{
      sphereCollisionRadius=0
      boundBone {boneName=dyn_feather_08}
      boxExtents=<5 5 5>
      constraintSetup {angularLimitsMax=<20 60 20>, angularLimitsMin=<-20 -5 -20>, angularTarget=<0 0 0>}
      localJointOffset=<0 0 -0.5>
     }]
    relativeSpaceBone {boneName=pelvis}
   }
   n24 = Anim Dynamics - Bone: dyn_feather_09(ComponentPose=@n23, Alpha=0.25, ExternalForce="0,0,0") {
    angularDampingOverride=0.98
    bGravityOverrideInSimSpace=true
    bOverrideAngularDamping=true
    bOverrideLinearDamping=true
    gravityScale=0.1
    lODThreshold=1
    linearDampingOverride=0.98
    numSolverIterationsPostUpdate=4
    numSolverIterationsPreUpdate=16
    simulationSpace=BoneRelative
    alphaBoolBlend {blendOption=Linear}
    boundBone {boneName=dyn_feather_09}
    physicsBodyDefinitions [{
      sphereCollisionRadius=0
      boundBone {boneName=dyn_feather_09}
      boxExtents=<5 5 5>
      constraintSetup {angularLimitsMax=<20 60 20>, angularLimitsMin=<-20 -5 -20>, angularTarget=<0 0 0>}
      localJointOffset=<0 0 0.5>
     }]
    relativeSpaceBone {boneName=pelvis}
   }
   n25 = Anim Dynamics - Bone: dyn_feather_10(ComponentPose=@n24, Alpha=0.25, ExternalForce="0,0,0") {
    angularDampingOverride=0.98
    bGravityOverrideInSimSpace=true
    bOverrideAngularDamping=true
    bOverrideLinearDamping=true
    gravityScale=0.1
    lODThreshold=1
    linearDampingOverride=0.98
    numSolverIterationsPostUpdate=4
    numSolverIterationsPreUpdate=16
    simulationSpace=BoneRelative
    alphaBoolBlend {blendOption=Linear}
    boundBone {boneName=dyn_feather_10}
    physicsBodyDefinitions [{
      sphereCollisionRadius=0
      boundBone {boneName=dyn_feather_10}
      boxExtents=<5 5 5>
      constraintSetup {angularLimitsMax=<20 60 20>, angularLimitsMin=<-20 -5 -20>, angularTarget=<0 0 0>}
      localJointOffset=<0 0 0.5>
     }]
    relativeSpaceBone {boneName=pelvis}
   }
   n26 = Anim Dynamics - Bone: dyn_feather_11(ComponentPose=@n25, Alpha=0.25, ExternalForce="0,0,0") {
    angularDampingOverride=0.98
    bGravityOverrideInSimSpace=true
    bOverrideAngularDamping=true
    bOverrideLinearDamping=true
    gravityScale=0.1
    lODThreshold=1
    linearDampingOverride=0.98
    numSolverIterationsPostUpdate=4
    numSolverIterationsPreUpdate=16
    simulationSpace=BoneRelative
    alphaBoolBlend {blendOption=Linear}
    boundBone {boneName=dyn_feather_11}
    physicsBodyDefinitions [{
      sphereCollisionRadius=0
      boundBone {boneName=dyn_feather_11}
      boxExtents=<5 5 5>
      constraintSetup {angularLimitsMax=<20 60 20>, angularLimitsMin=<-20 -5 -20>, angularTarget=<0 0 0>}
      localJointOffset=<0 0 0.5>
     }]
    relativeSpaceBone {boneName=pelvis}
   }
   n27 = RigidBody(ComponentPose=@n26, Alpha=0.8) {
    bClampLinearTranslationLimitToRefPose=true
    bOverrideWorldGravity=true
    overridePhysicsAsset=&M_MED_Tender_Defender_FaceACC_Physics
    simulationSpace=BaseBoneSpace
    alphaBoolBlend {blendOption=Linear}
    alphaScaleBiasClamp {bClampResult=true, clampMax=0.8}
    baseBoneRef {boneName=root}
    componentAppliedLinearAccClamp=<500 500 500>
    componentLinearAccScale=<0.3 0.3 0.3>
    componentLinearVelScale=<0.2 0.2 0.2>
    overrideWorldGravity {z=150}
    simSpaceSettings {maxAngularAcceleration=250, maxAngularVelocity=3, maxLinearAcceleration=150, maxLinearVelocity=150, velocityScaleZ=0, worldAlpha=1}
   }
   Output Pose(Result=Blend Poses by bool(
     BlendPose_0=Component To Local(ComponentPose=@n7)
     BlendPose_1=Component To Local(ComponentPose=@n27)
     BlendTime_0=0
     BlendTime_1=0
     bActiveValue=Get GetPlayerListUsesCustomCharacterParts() -> (GetPlayerListUsesCustomCharacterParts:bool)
    ) {blendType=Linear}) {layerGroup=None}
   Save cached pose 'PreScale Cache'(Pose=Copy Pose From Mesh() {bUseAttachedParent=true}) {cacheName=PreScale Cache}
   n28 = Anim Dynamics - Bone: dyn_hair_01_a(Alpha=0.25, ExternalForce="0,0,0") {
    alpha=0.9
    angularDampingOverride=0.99
    angularSpringConstant=80
    bAngularSpring=true
    bGravityOverrideInSimSpace=true
    bOverrideAngularDamping=true
    gravityScale=0.85
    numSolverIterationsPostUpdate=4
    numSolverIterationsPreUpdate=16
    simulationSpace=BoneRelative
    alphaBoolBlend {blendOption=Linear}
    boundBone {boneName=dyn_hair_01_a}
    physicsBodyDefinitions [{
      sphereCollisionRadius=0
      boundBone {boneName=dyn_hair_01_a}
      boxExtents=<15 5 5>
      constraintSetup {angularTargetAxis=AxisZ, twistAxis=AxisY, angularLimitsMax=<15 0 89>, angularLimitsMin=<-15 0 -89>, angularTarget=<0 0 1>}
      localJointOffset=<0 2 6>
     }]
    relativeSpaceBone {boneName=root}
   }
   n29 = Anim Dynamics - Bone: dyn_hair_01_b(ComponentPose=@n28, Alpha=0.25, ExternalForce="0,0,0") {
    alpha=0.9
    angularDampingOverride=0.99
    angularSpringConstant=80
    bAngularSpring=true
    bGravityOverrideInSimSpace=true
    bOverrideAngularDamping=true
    gravityScale=0.85
    numSolverIterationsPostUpdate=4
    numSolverIterationsPreUpdate=16
    simulationSpace=BoneRelative
    alphaBoolBlend {blendOption=Linear}
    boundBone {boneName=dyn_hair_01_b}
    physicsBodyDefinitions [{
      sphereCollisionRadius=0
      boundBone {boneName=dyn_hair_01_b}
      boxExtents=<15 5 5>
      constraintSetup {angularTargetAxis=AxisZ, twistAxis=AxisY, angularLimitsMax=<15 0 89>, angularLimitsMin=<-15 0 -89>, angularTarget=<0 0 1>}
      localJointOffset=<0 2 6>
     }]
    relativeSpaceBone {boneName=root}
   }
   n30 = Anim Dynamics - Bone: dyn_hair_01_c(ComponentPose=@n29, Alpha=0.25, ExternalForce="0,0,0") {
    alpha=0.9
    angularDampingOverride=0.9999
    angularSpringConstant=80
    bAngularSpring=true
    bGravityOverrideInSimSpace=true
    bOverrideAngularDamping=true
    gravityScale=0.85
    simulationSpace=BoneRelative
    alphaBoolBlend {blendOption=Linear}
    boundBone {boneName=dyn_hair_01_c}
    physicsBodyDefinitions [{
      sphereCollisionRadius=0
      boundBone {boneName=dyn_hair_01_c}
      boxExtents=<15 5 5>
      constraintSetup {angularTargetAxis=AxisZ, twistAxis=AxisY, angularLimitsMax=<15 0 89>, angularLimitsMin=<-15 0 -89>, angularTarget=<0 0 1>}
      localJointOffset=<0 1 4>
     }]
    relativeSpaceBone {boneName=neck_01}
   }
   n31 = Anim Dynamics - Bone: dyn_hair_01(ComponentPose=@n30, Alpha=0) {
    alpha=0.9
    angularSpringConstant=80
    bAngularSpring=true
    bGravityOverrideInSimSpace=true
    numSolverIterationsPostUpdate=4
    numSolverIterationsPreUpdate=16
    simulationSpace=BoneRelative
    alphaBoolBlend {blendOption=Linear}
    boundBone {boneName=dyn_hair_01}
    physicsBodyDefinitions [{
      sphereCollisionRadius=0
      boundBone {boneName=dyn_hair_01}
      boxExtents=<15 5 5>
      constraintSetup {angularTargetAxis=AxisZ, twistAxis=AxisY, angularLimitsMax=<15 0 89>, angularLimitsMin=<-15 0 -89>, angularTarget=<0 0 1>}
      localJointOffset=<0 0 3>
     }]
    relativeSpaceBone {boneName=root}
   }
   n32 = Anim Dynamics - Bone: dyn_hair_03(ComponentPose=@n31, Alpha=0, ExternalForce="0,0,0") {
    alpha=0.9
    angularSpringConstant=80
    bAngularSpring=true
    bGravityOverrideInSimSpace=true
    numSolverIterationsPostUpdate=4
    numSolverIterationsPreUpdate=16
    simulationSpace=BoneRelative
    alphaBoolBlend {blendOption=Linear}
    boundBone {boneName=dyn_hair_03}
    physicsBodyDefinitions [{
      sphereCollisionRadius=0
      boundBone {boneName=dyn_hair_03}
      boxExtents=<15 5 5>
      constraintSetup {angularTargetAxis=AxisZ, twistAxis=AxisY, angularLimitsMax=<15 0 89>, angularLimitsMin=<-15 0 -89>, angularTarget=<0 0 1>}
      localJointOffset=<0 0 3>
     }]
    relativeSpaceBone {boneName=root}
   }
   Reroute Node(InputPin:Component Space Pose Link Structure=@n32) -> (OutputPin:Component Space Pose Link Structure)
  ]}
 EventGraph {nodes [
   Event BlueprintUpdateAnimation()
   > Set Vehicle Off(Vehicle Off:double=SelectFloat(A=0, B=1, bPickA=Get bIsInVehicle() -> (bIsInVehicle:bool))) -> (Output_Get:double)
   Event BlueprintInitializeAnimation()
   > Set GetPlayerListUsesCustomCharacterParts(GetPlayerListUsesCustomCharacterParts:bool=GetPlaylistUsesCustomCharacterParts!() -> (ReturnValue:bool)) -> (Output_Get:bool)
  ]}
]
variables [
 GetPlayerListUsesCustomCharacterParts {type=boolean}
 Vehicle Off {type=number}
]
